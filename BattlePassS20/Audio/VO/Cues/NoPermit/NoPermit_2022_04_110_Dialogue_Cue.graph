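# Sound cue that plays a Season 20 Battle Pass voice-over line spoken by The Imagined.
# A single Dialogue Player node resolves the NoPermit dialogue wave using that speaker context, with no targets.
# The cue uses the S20 quest attenuation, the narrator VO sound class and a boosted volume.
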SoundCue /BattlePassS20/Audio/VO/Cues/NoPermit/NoPermit_2022_04_110_Dialogue_Cue
root=DialoguePlayer_0
nodes [{
  id=DialoguePlayer_0
  properties {dialogueWaveParameter {
    dialogueWave=/BattlePassS20/Audio/VO/Waves/NoPermit/NoPermit_2022_04_110_Dialogue
    context {speaker=/BattlePassS20/Audio/VO/TheImagined_DV, targets []}
   }}
 }]
properties {
 attenuationSettings=/BattlePassS20/Audio/VO/S20_Quest_Att
 maxDistance=2097152
 soundClassObject=/Game/Sounds/FortSoundClassesAndModes/Fort_VO_Narrator
 volumeMultiplier=2
 concurrencySet [null]
}
